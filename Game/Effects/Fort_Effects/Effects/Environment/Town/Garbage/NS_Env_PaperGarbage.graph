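# Niagara environment effect scattering a User-sized batch of paper garbage meshes over the landscape, which a camera-relative repulsion point stirs up.
# GPU particles spawn within a radius, sample landscape height and normal, and gate wind, noise, spin and drag on terrain-contact and latched active flags.
# Aging depends on velocity, with active pieces shrinking late in life.
# Only InteractiveConfetti001, which adds aerodynamic drag, is enabled; the similar InteractiveConfetti is disabled.
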NiagaraSystem /Game/Effects/Fort_Effects/Effects/Environment/Town/Garbage/NS_Env_PaperGarbage
dependencies {
 usedDataInterfaces [&NiagaraDataInterfaceCamera, &NiagaraDataInterfaceCurve]
 usedDynamicInputs []
 usedModules [&SystemState]
}
emitters {
 InteractiveConfetti {
  bEnabled=false
  simTarget=GPUComputeSim
  emitterUpdateScript [
   /SetVariables() {isSetParameters=true, module=SetVariables_#1}
   /SetVariables(
    Emitter.CameraForward:Vector3f=&ReturnCameraVector(Camera Property=::NewEnumerator1, Coordinate Space=::Simulation, CameraQuery=$Emitter.CameraRead)
    Emitter.CameraPosition:Vector3f=&ReturnCameraVector(Camera Property=::NewEnumerator0, Coordinate Space=::World, CameraQuery=$Emitter.CameraRead)
    Emitter.CameraRight:Vector3f=&ReturnCameraVector(Camera Property=::NewEnumerator3, Coordinate Space=::Simulation, CameraQuery=$Emitter.CameraRead)
    Emitter.CameraUp:Vector3f=&ReturnCameraVector(Camera Property=::NewEnumerator2, Coordinate Space=::Simulation, CameraQuery=$Emitter.CameraRead)
   ) {isSetParameters=true, module=SetVariables_#2}
   &EmitterState(
    Enable Distance Culling=true
    Life Cycle Mode=::NewEnumerator1
    Loop Behavior=::NewEnumerator1
    Loop Duration Mode=::NewEnumerator1
    Max Distance Response=::SleepAndClearParticles
    MaxDistance=6000
    Module.bUseMaxDistance=true
    Module.bUseMinDistance=false
    Reset Age On Awaken=true
    Scalability Mode=::NewEnumerator1
    UseLoopDelay=false
   )
   &SpawnRate(
    SpawnRate=&FloatFromCurve(
     FloatCurve=DataInterface() {propertyValues {Curve {keys [<at 0 0>,<at 0.25 2>,<at 0.5 0>]}}}
     CurveIndex=$Emitter.Age
     Scale Curve=$User.SpawnAmount
    )
    Use Spawn Probability=false
   )
  ]
  particleSpawnScript [
   &InitializeParticle(
    Lifetime Max=5
    Lifetime Min=3
    Lifetime Mode=::NewEnumerator1
    Mass Max=6
    Mass Min=3
    Mass Mode=::NewEnumerator3
    Mesh Scale Mode=::NewEnumerator1
    Mesh Uniform Scale Max:NiagaraFloat=$User.MaxSize
    Mesh Uniform Scale Min:NiagaraFloat=$User.MinSize
    UsePositionOffset=false
    Write Material Random=true
   ) {module=InitializeParticle001}
   /SetVariables() {isSetParameters=true, module=SetVariables_#3}
   /SetVariables(Particles.MeshIndex=&RandomRangeInt(Minimum=0, Maximum=3)) {isSetParameters=true, module=SetVariables_#4}
   /SetVariables(Particles.Terrain Offset:NiagaraFloat=$User.TerrainOffset) {isSetParameters=true, module=SetVariables_#5}
   /Niagara/Modules/Spawn/Location/SphereLocation(MaskSpawnGroup=false, Module.Override Randomness=false, Sphere Radius=$User.RadiusSpread, Surface Only=false)
   &self:Sample_Landscape() {module=Sample_Landscape}
   /SetVariables(Particles.Position=&Add_Vector(
     A=Linked() {linkedVariable {name=Particles.Position, type=NiagaraPosition}}
     B=&MakeVector(X=0, Y=0, Z=$Particles.Terrain Offset)
    )) {isSetParameters=true, module=SetVariables_#6}
   &AddVelocity()
   &CameraOffset(Camera Offset Amount=15)
   &SubUVAnimation(End Frame=36, SubUV Animation Mode:ENiagara_SubUVLookupMode=::NewEnumerator2) {module=SubUVAnimation001}
   /SetVariables(Particles.SubImageIndex=&RandomRangeFloat(Minimum=0, Maximum=35)) {isSetParameters=true, module=SetVariables_#7}
   &InitialMeshOrientation(
    Mesh Orientation Mode=::NewEnumerator2
    Rotation=&RandomRangeVector(Minimum=<0 0 0>, Maximum=<0 0 1>)
    Use Rotation Vector=true
   )
   &KillParticles(Kill Particles=&SetBoolByIntComparison(
     A=$Engine.Emitter.TotalSpawnedParticles
     B=&MakeIntFromFloat(Float=$User.SpawnAmount)
     Comparison Type=::NewEnumerator0
    ))
  ]
  particleUpdateScript [
   &self:Sample_Landscape_UPdate(Bias:NiagaraFloat=$Particles.Terrain Offset, New Landscape Sample:NiagaraDataInterfaceLandscape=$User.Landscape) {module=Sample_Landscape_UPdate}
   &ScaleColor(
    Scale Alpha=&FloatFromCurve(FloatCurve=DataInterface() {propertyValues {Curve {keys [<at 0.9 5>,<at 1 0>]}}})
    Scale RGB=&MakeVectorFromLinearColorRGB(Color=&MultiplyLinearColor(Linear Color=$Particles.Initial.Color, Scale Factor=&MakeVector4_FromFloat(Float=1)))
    ScaleA=true
    ScaleRGB=true
   )
   /SetVariables(Particles.Has Collided With Terrain:NiagaraBool=&SetBoolByFloatComparison(
     A=&Add_Float(A=$Output.Sample_Landscape_UPdate.Terrain Height, B=20)
     B=&MakeFloatFromVector(Vector=Linked() {linkedVariable {name=Particles.Position, type=NiagaraPosition}}, Channel=::NewEnumerator2)
    )) {isSetParameters=true, module=SetVariables_#8}
   /Niagara/Modules/Update/Forces/PointAttractionForce(
    AttractionStrength=-85
    AttractorPosition:Vector3f=&Add_Vector(
     A=$Emitter.CameraPosition
     B=&Add_Vector(
      A=&Add_Vector(
       A=&Multiply_VectorByFloat(Vector=$Particles.Landscape Normal, Float=-130)
       B=&Multiply_VectorByFloat(Vector=$Emitter.CameraForward, Float=300)
      )
      B=&Multiply_VectorByFloat(Vector=$Emitter.CameraRight, Float=-50)
     )
    )
    Falloff Exponent=1
    Module.Kill Within Radius=false
    Module.Use Falloff=false
   )
   /SetVariables(Particles.Normalized Velocity:Vector3f=&NormalizeVector(Vector To Normalize=$Particles.Velocity)) {isSetParameters=true, module=SetVariables_#9}
   /SetVariables(Particles.Is Active:NiagaraBool=&Boolean_OrOperation(
     Bool1=$Particles.Is Active
     Bool2=&SetBoolByFloatComparison(
      A=&MakeFloatFromVector(Vector=Linked() {linkedVariable {name=Particles.Position, type=NiagaraPosition}}, Channel=::NewEnumerator2)
      B=$Output.Sample_Landscape_UPdate.Terrain Height
      Comparison Type=::NewEnumerator5
     )
    )) {isSetParameters=true, module=SetVariables_#10}
   &WindForce(
    Wind Speed=&RandomRangeVector(Minimum=<250 -250 50>, Maximum=<800 250 100>)
    Wind Speed Scale=&MakeCustomFloatFromBool(Bool=$Particles.Has Collided With Terrain, True Float=0, False Float=2)
   )
   &CurlNoiseForce(
    Noise Frequency=10
    Noise Strength=&MakeCustomFloatFromBool(
     Bool=$Particles.Is Active
     True Float=&FloatFromCurve(FloatCurve=DataInterface() {propertyValues {Curve {keys [<at 0.25 1>,<at 0.5 0>]}}}, Scale Curve=500)
     False Float=0
    )
    Pan Noise=true
    Pan Noise Field=<0 0 1>
   ) {module=CurlNoiseForce001}
   &MeshRotationForce(Rotation=&Multiply_VectorByFloat(
     Vector=&RandomRangeVector(Minimum=<25 25 25>, Maximum=<50 50 50>)
     Float=&MakeCustomFloatFromBool(Bool=$Particles.Is Active, True Float=1, False Float=0)
    ))
   &Drag(
    Drag=&MakeCustomFloatFromBool(
     Bool=$Particles.Is Active
     True Float=&FloatFromCurve(FloatCurve=DataInterface() {propertyValues {Curve {keys [<at 0.1 0>,<at 0.25 0.95>,<at 0.5 1>,<at 0.7 0>]}}})
     False Float=0
    )
    Rotational Drag:NiagaraFloat=&FloatFromCurve(
     FloatCurve=DataInterface() {propertyValues {Curve {keys [{tangentMode=RCTM_User, time=0, value=0}, {tangentMode=RCTM_User, time=1, value=1}]}}}
     Scale Curve=0.05
    )
    Use Linear Drag=true
    Use Rotational Drag=true
   )
   &GravityForce(Gravity=&MakeVector(
     X=0
     Y=0
     Z=&FloatFromCurve(
      FloatCurve=DataInterface() {propertyValues {Curve {keys [
          <at 0 0>
          {interpMode=RCIM_Cubic, tangentMode=RCTM_User, time=0.02, value=1}
          <at 0.1 0>
          {interpMode=RCIM_Cubic, tangentMode=RCTM_User, time=0.2, value=-0.9}
          <at 0.35 -0.25>
          <at 0.8 0.1>
          <at 1 0.1>
         ]}}}
      Scale Curve=600
     )
    )) {module=GravityForce001}
   /SetVariables(Particles.Sprite Orientation:Vector3f=&Lerp_Vector(
     A=$Particles.Landscape Normal
     B=&ReturnCameraVector(Camera Property=::NewEnumerator4, Coordinate Space=::Simulation, Origin Position=$Particles.Position, CameraQuery=$Emitter.CameraRead)
     Alpha=&MakeCustomFloatFromBool(Bool=$Particles.Is Active, True Float=1, False Float=0)
    )) {isSetParameters=true, module=SetVariables_#11}
   &SolveForcesAndVelocity(Clamp Velocity=true, Limit Acceleration=false)
   &self:ScratchModule_03(Terrain Height:NiagaraFloat=$Output.Sample_Landscape_UPdate.Terrain Height, Threshold=40) {module=ScratchModule_03}
   &ParticleState(DeltaTime:NiagaraFloat=&MakeCustomFloatFromBool(
     Bool=&SetBoolByFloatComparison(A=&VectorLength(Vector=$Particles.Velocity), B=1)
     True Float=$Engine.DeltaTime
     False Float=0
    ))
   &ScaleMeshSize(Scale Factor=&Multiply_VectorByFloat(
     Vector=<1 1 1>
     Float=&MakeCustomFloatFromBool(
      Bool=$Particles.Is Active
      True Float=&FloatFromCurve(FloatCurve=DataInterface() {propertyValues {Curve {keys [<at 0.75 1 RCIM_Cubic>,<at 1 0 RCIM_Cubic>]}}})
      False Float=1
     )
    ))
   &Color(Color=&MakeCustomLinearColorFromBool(Bool=$Particles.Is Close to Ground, True Color=<rgba 0 0.2682 1 1>, False Color=<rgba 0 1 0.2985 1>)) {enabled=false}
  ]
  rendererClasses [&NiagaraMeshRendererProperties]
  renderers [{rendererClass=&NiagaraMeshRendererProperties, rendererIndex=0}]
 }
 InteractiveConfetti001 {
  simTarget=GPUComputeSim
  emitterUpdateScript [
   /SetVariables() {isSetParameters=true, module=SetVariables_#1}
   /SetVariables(
    Emitter.CameraForward:Vector3f=&ReturnCameraVector(Camera Property=::NewEnumerator1, Coordinate Space=::Simulation, CameraQuery=$Emitter.CameraRead)
    Emitter.CameraPosition:Vector3f=&ReturnCameraVector(Camera Property=::NewEnumerator0, Coordinate Space=::World, CameraQuery=$Emitter.CameraRead)
    Emitter.CameraRight:Vector3f=&ReturnCameraVector(Camera Property=::NewEnumerator3, Coordinate Space=::Simulation, CameraQuery=$Emitter.CameraRead)
    Emitter.CameraUp:Vector3f=&ReturnCameraVector(Camera Property=::NewEnumerator2, Coordinate Space=::Simulation, CameraQuery=$Emitter.CameraRead)
   ) {isSetParameters=true, module=SetVariables_#2}
   &EmitterState(
    Enable Distance Culling=true
    Life Cycle Mode=::NewEnumerator1
    Loop Behavior=::NewEnumerator1
    Loop Duration Mode=::NewEnumerator1
    Max Distance Response=::SleepAndClearParticles
    MaxDistance=6000
    Module.bUseMaxDistance=true
    Module.bUseMinDistance=false
    Reset Age On Awaken=true
    Scalability Mode=::NewEnumerator1
    UseLoopDelay=false
   )
   &SpawnRate(
    SpawnRate=&FloatFromCurve(
     FloatCurve=DataInterface() {propertyValues {Curve {keys [<at 0 0>,<at 0.25 2>,<at 0.5 0>]}}}
     CurveIndex=$Emitter.Age
     Scale Curve=$User.SpawnAmount
    )
    Use Spawn Probability=false
   )
  ]
  particleSpawnScript [
   &InitializeParticle(
    Lifetime Max=5
    Lifetime Min=3
    Lifetime Mode=::NewEnumerator1
    Mass Max=6
    Mass Min=3
    Mass Mode=::NewEnumerator3
    Mesh Scale Mode=::NewEnumerator1
    Mesh Uniform Scale Max:NiagaraFloat=$User.MaxSize
    Mesh Uniform Scale Min:NiagaraFloat=$User.MinSize
    UsePositionOffset=false
    Write Material Random=true
   ) {module=InitializeParticle001}
   /SetVariables() {isSetParameters=true, module=SetVariables_#3}
   /SetVariables(Particles.MeshIndex=&RandomRangeInt(Minimum=0, Maximum=3)) {isSetParameters=true, module=SetVariables_#4}
   /SetVariables(Particles.Terrain Offset:NiagaraFloat=$User.TerrainOffset) {isSetParameters=true, module=SetVariables_#5}
   &SphereLocation(MaskSpawnGroup=false, Module.Override Randomness=false, Sphere Radius=$User.RadiusSpread, Surface Only=false)
   &self:Sample_Landscape() {module=Sample_Landscape}
   /SetVariables(Particles.Position=&Add_Vector(
     A=Linked() {linkedVariable {name=Particles.Position, type=NiagaraPosition}}
     B=&MakeVector(X=0, Y=0, Z=$Particles.Terrain Offset)
    )) {isSetParameters=true, module=SetVariables_#6}
   &AddVelocity()
   &CameraOffset(Camera Offset Amount=15)
   &SubUVAnimation(End Frame=36, SubUV Animation Mode:ENiagara_SubUVLookupMode=::NewEnumerator2) {module=SubUVAnimation001}
   /SetVariables(Particles.SubImageIndex=&RandomRangeFloat(Minimum=0, Maximum=35)) {isSetParameters=true, module=SetVariables_#7}
   &InitialMeshOrientation(
    Mesh Orientation Mode=::NewEnumerator2
    Rotation=&RandomRangeVector(Minimum=<0 0 0>, Maximum=<0 0 1>)
    Use Rotation Vector=true
   )
   &KillParticles(Kill Particles=&SetBoolByIntComparison(
     A=$Engine.Emitter.TotalSpawnedParticles
     B=&MakeIntFromFloat(Float=$User.SpawnAmount)
     Comparison Type=::NewEnumerator0
    ))
  ]
  particleUpdateScript [
   &self:Sample_Landscape_UPdate(Bias:NiagaraFloat=$Particles.Terrain Offset, New Landscape Sample:NiagaraDataInterfaceLandscape=$User.Landscape) {module=Sample_Landscape_UPdate}
   &ScaleColor(
    Scale Alpha=&FloatFromCurve(FloatCurve=DataInterface() {propertyValues {Curve {keys [<at 0.9 5>,<at 1 0>]}}})
    Scale RGB=&MakeVectorFromLinearColorRGB(Color=&MultiplyLinearColor(Linear Color=$Particles.Initial.Color, Scale Factor=&MakeVector4_FromFloat(Float=1)))
    ScaleA=true
    ScaleRGB=true
   )
   /SetVariables(Particles.Has Collided With Terrain:NiagaraBool=&SetBoolByFloatComparison(
     A=&Add_Float(A=$Output.Sample_Landscape_UPdate.Terrain Height, B=20)
     B=&MakeFloatFromVector(Vector=Linked() {linkedVariable {name=Particles.Position, type=NiagaraPosition}}, Channel=::NewEnumerator2)
    )) {isSetParameters=true, module=SetVariables_#8}
   &PointAttractionForce(
    Attraction Radius:NiagaraFloat=&RandomRangeFloat(Minimum=55, Maximum=80)
    AttractionStrength=-55
    AttractorPosition:Vector3f=&Add_Vector(
     A=$Emitter.CameraPosition
     B=&Add_Vector(
      A=&Add_Vector(
       A=&Multiply_VectorByFloat(Vector=$Particles.Landscape Normal, Float=-140)
       B=&Multiply_VectorByFloat(Vector=$Emitter.CameraForward, Float=300)
      )
      B=&Multiply_VectorByFloat(Vector=$Emitter.CameraRight, Float=-50)
     )
    )
    Falloff Exponent=1
    Module.Kill Within Radius=false
    Module.Use Falloff=false
   )
   /SetVariables(Particles.Normalized Velocity:Vector3f=&NormalizeVector(Vector To Normalize=$Particles.Velocity)) {isSetParameters=true, module=SetVariables_#9}
   /SetVariables(Particles.Is Active:NiagaraBool=&Boolean_OrOperation(
     Bool1=$Particles.Is Active
     Bool2=&SetBoolByFloatComparison(
      A=&MakeFloatFromVector(Vector=Linked() {linkedVariable {name=Particles.Position, type=NiagaraPosition}}, Channel=::NewEnumerator2)
      B=$Output.Sample_Landscape_UPdate.Terrain Height
      Comparison Type=::NewEnumerator5
     )
    )) {isSetParameters=true, module=SetVariables_#10}
   &WindForce(
    Wind Speed=&RandomRangeVector(Minimum=<250 -250 50>, Maximum=<800 250 100>)
    Wind Speed Scale=&MakeCustomFloatFromBool(Bool=$Particles.Has Collided With Terrain, True Float=0, False Float=0.3)
   )
   &CurlNoiseForce(
    Noise Frequency=10
    Noise Strength=&MakeCustomFloatFromBool(
     Bool=$Particles.Is Active
     True Float=&FloatFromCurve(FloatCurve=DataInterface() {propertyValues {Curve {keys [<at 0.25 1>,<at 0.5 0>]}}}, Scale Curve=500)
     False Float=0
    )
    Pan Noise=true
    Pan Noise Field=<0 0 1>
   ) {module=CurlNoiseForce001}
   &MeshRotationForce(Rotation=&Multiply_VectorByFloat(
     Vector=&RandomRangeVector(Minimum=<2 2 2>, Maximum=<25 25 25>)
     Float=&MakeCustomFloatFromBool(Bool=$Particles.Is Active, True Float=0.2, False Float=0)
    ))
   &Drag(
    Drag=&MakeCustomFloatFromBool(
     Bool=$Particles.Is Active
     True Float=&FloatFromCurve(FloatCurve=DataInterface() {propertyValues {Curve {keys [<at 0.1 0.5>,<at 0.25 0.95>,<at 0.5 1>,<at 0.7 0.5>]}}}, Scale Curve=0.1)
     False Float=0
    )
    Rotational Drag:NiagaraFloat=&FloatFromCurve(
     FloatCurve=DataInterface() {propertyValues {Curve {keys [{tangentMode=RCTM_User, time=0, value=0}, {tangentMode=RCTM_User, time=1, value=1}]}}}
     Scale Curve=0.2
    )
    Use Linear Drag=true
    Use Rotational Drag=true
   )
   &GravityForce(Gravity=&MakeVector(
     X=0
     Y=0
     Z=&FloatFromCurve(
      FloatCurve=DataInterface() {propertyValues {Curve {keys [
          {arriveTangent=0.2435, interpMode=RCIM_Cubic, leaveTangent=0.2435, tangentMode=RCTM_User, time=0, value=0}
          {arriveTangent=1.1179, interpMode=RCIM_Cubic, leaveTangent=1.1179, tangentMode=RCTM_User, time=0.3, value=0.125}
          {arriveTangent=0.0224, interpMode=RCIM_Cubic, leaveTangent=0.0224, tangentMode=RCTM_User, time=1, value=1}
         ]}}}
      Scale Curve=&RandomRangeFloat(Minimum=-980, Maximum=-600)
     )
    )) {module=GravityForce001}
   &AerodynamicDrag(
    Aerodynamic Drag=0.6
    Aerodynamic Rotational Drag=2
    Drag Shape Curve=DataInterface() {propertyValues {Curve {keys [
        {interpMode=RCIM_Cubic, tangentMode=RCTM_User, time=0, value=0.1}
        {arriveTangent=2.0799, interpMode=RCIM_Cubic, leaveTangent=2.0799, tangentMode=RCTM_User, time=1, value=1}
       ]}}}
    Lift Shape Curve=DataInterface() {propertyValues {Curve {keys [
        {interpMode=RCIM_Cubic, tangentMode=RCTM_User, time=0, value=0}
        {interpMode=RCIM_Cubic, tangentMode=RCTM_User, time=0.5, value=1}
        {tangentMode=RCTM_User, time=1, value=0}
       ]}}}
    Pivot Offset Maximum:Vector3f=<1 1 1>
    Pivot Offset Minimum Length:Vector3f=<250 0 500>
    Rotation Strength:NiagaraFloat=&RandomRangeFloat(Minimum=2, Maximum=8)
    Use Linear Drag=true
    Use Rotational Drag=true
   )
   /SetVariables(Particles.Sprite Orientation:Vector3f=&Lerp_Vector(
     A=$Particles.Landscape Normal
     B=&ReturnCameraVector(Camera Property=::NewEnumerator4, Coordinate Space=::Simulation, Origin Position=$Particles.Position, CameraQuery=$Emitter.CameraRead)
     Alpha=&MakeCustomFloatFromBool(Bool=$Particles.Is Active, True Float=1, False Float=0)
    )) {isSetParameters=true, module=SetVariables_#11}
   &SolveForcesAndVelocity(Clamp Velocity=true, Limit Acceleration=false)
   &self:ScratchModule_03(Terrain Height:NiagaraFloat=$Output.Sample_Landscape_UPdate.Terrain Height, Threshold=40) {module=ScratchModule_03}
   &ParticleState(DeltaTime:NiagaraFloat=&MakeCustomFloatFromBool(
     Bool=&SetBoolByFloatComparison(A=&VectorLength(Vector=$Particles.Velocity), B=1)
     True Float=$Engine.DeltaTime
     False Float=0
    ))
   &ScaleMeshSize(Scale Factor=&Multiply_VectorByFloat(
     Vector=<1 1 1>
     Float=&MakeCustomFloatFromBool(
      Bool=$Particles.Is Active
      True Float=&FloatFromCurve(FloatCurve=DataInterface() {propertyValues {Curve {keys [<at 0.8 1 RCIM_Cubic>, {arriveTangent=-12.063, interpMode=RCIM_Cubic, leaveTangent=-12.063, tangentMode=RCTM_User, time=1, value=0}]}}})
      False Float=1
     )
    ))
   &AlignSpriteToMeshOrientation(SetAlignment=true, SetFacing=true)
   &Color(Color=&MakeCustomLinearColorFromBool(Bool=$Particles.Is Close to Ground, True Color=<rgba 0 0.2682 1 1>, False Color=<rgba 0 1 0.2985 1>)) {enabled=false}
   &DynamicMaterialParameters(
    Index 0 Param 1=0.02
    Index 0 Param 4=&FloatFromCurve(FloatCurve=DataInterface() {propertyValues {Curve {keys [
         {arriveTangent=4.1785, interpMode=RCIM_Cubic, leaveTangent=4.1785, tangentMode=RCTM_User, time=0, value=0}
         {arriveTangent=0.1622, interpMode=RCIM_Cubic, leaveTangent=0.1622, tangentMode=RCTM_User, time=0.1, value=1}
         {arriveTangent=0.0224, interpMode=RCIM_Cubic, leaveTangent=0.0224, tangentMode=RCTM_User, time=0.5, value=0.95}
         <at 0.8 0>
        ]}}})
    Module.Param0WriteEnabled=false
    Module.Param1WriteEnabled=false
    Module.Param2WriteEnabled=false
    Module.Param3WriteEnabled=true
   ) {module=DynamicMaterialParameters001}
  ]
  rendererClasses [&NiagaraMeshRendererProperties]
  renderers [{rendererClass=&NiagaraMeshRendererProperties, rendererIndex=0}]
 }
}
summary {
 systemName=NS_Env_PaperGarbage
 emitters [
  {bEnabled=false, emitterName=InteractiveConfetti, simTarget=GPUComputeSim, rendererClasses [&NiagaraMeshRendererProperties]}
  {emitterName=InteractiveConfetti001, simTarget=GPUComputeSim, rendererClasses [&NiagaraMeshRendererProperties]}
 ]
 userVariables [
  User.Landscape {
   type=NiagaraDataInterfaceLandscape
   description=""
   defaultValue {
    struct=NiagaraExt_VariableValue_DataInterface
    value {
     dataInterface=&self:NiagaraDataInterfaceLandscape_0
     dataInterfaceClass=&NiagaraDataInterfaceLandscape
    }
   }
  }
  User.MaxSize {type=NiagaraFloat, defaultValue=1, description=""}
  User.MinSize {type=NiagaraFloat, defaultValue=0.5, description=""}
  User.RadiusSpread {type=NiagaraFloat, defaultValue=100, description=""}
  User.SpawnAmount {type=NiagaraFloat, defaultValue=15, description=""}
  User.TerrainOffset {type=NiagaraFloat, defaultValue=5, description=""}
 ]
}
